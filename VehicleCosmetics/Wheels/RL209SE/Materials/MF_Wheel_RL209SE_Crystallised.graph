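# Material function for the RL209SE Crystallised wheel, overriding MF_WheelParent_Simple with a glossy, fully metallic, faceted crystal look.
# Facet and normal textures perturb the reflection vector sampling a rotating environment texture, adding fake reflections over mask-driven rim and accent colours.
# Vertex red separates rim from tire, whose colour shimmers through a panning gradient distorted by its normal map.
# The combined colour also feeds an emissive reroute; several parent-blend lerps are left unconnected.
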MaterialFunction /VehicleCosmetics/Wheels/RL209SE/Materials/MF_Wheel_RL209SE_Crystallised
graph {nodes [
  n1 = Power(Base=NamedRerouteUsage()) {constExponent=3}
  n2 = Multiply(A=ScalarParameter() {defaultValue=0.2, parameterName=Rim_shiftspeed}, B=Time())
  n3 = Divide(A=Add(A=Sine(None=@n2)))
  n4 = Panner(Coordinate=TextureCoordinate()) {speedX=0.5}
  n5 = TextureSampleParameter2D() {
   group=1_Input Textures
   parameterName=Tire_Normal
   samplerType=SAMPLERTYPE_LinearColor
   texture=/VehicleCosmetics/Wheels/RL209/Textures/T_Tire_RL209_N
  }
  n6 = Power(Base=@n5.B, Exp=Constant() {r=5})
  n7 = Multiply(A=OneMinus(None=Clamp(None=@n6)), B=ScalarParameter() {defaultValue=0.6, parameterName=shimmerSelfDistort})
  n8 = VertexColor()
  n9 = LinearInterpolate(
   A=Frac(None=Add(A=@n1, B=@n3))
   B=TextureSample(UVs=Subtract(A=@n4, B=@n7)) {desc=tire gradient, samplerType=SAMPLERTYPE_LinearGrayscale, texture=/VehicleCosmetics/Wheels/RL209SE/Textures/T_GradientV_T}.R
   Alpha=@n8.R
  )
  Add(A=NamedRerouteUsage(), B=@n9)
  Constant3Vector() {constant=<rgba 0.7188 0.6241 0.5203 1>}
  n10 = MaterialFunctionCall() {materialFunction=&MF_WheelParent_Simple}
  n11 = GetMaterialAttributes(None=@n10.MaterialAttributes) {attributeGetTypes [
    #1
    #2
    #3
    #4
    #5
   ]}
  n12 = SetMaterialAttributes!(
   MaterialAttributes=@n11.MaterialAttributes
   EmissiveColor=NamedRerouteUsage()
   BaseColor=NamedRerouteUsage()
   Metallic=Constant() {r=1}
   Normal=NamedRerouteUsage()
   Specular=NamedRerouteUsage()
   Roughness=Constant()
  ) {attributeSetTypes [
    #1
    #2
    #3
    #4
    #5
    #6
   ]}
  FunctionOutput(None=@n12)
  n13 = ComponentMask(None=@n10.RimRGBAMask) {r=true}
  LinearInterpolate(A=LinearInterpolate(A=@n11.EmissiveColor, B=NamedRerouteUsage(), Alpha=@n13) {constB=0}, B=@n11.EmissiveColor, Alpha=NamedRerouteUsage())
  LinearInterpolate(A=LinearInterpolate(A=@n11.Normal, B=NamedRerouteUsage(), Alpha=@n13), B=@n11.Normal, Alpha=NamedRerouteUsage())
  LinearInterpolate(A=LinearInterpolate(A=@n11.Specular, B=NamedRerouteUsage(), Alpha=@n13), B=@n11.Specular, Alpha=NamedRerouteUsage())
  LinearInterpolate(A=LinearInterpolate(A=@n11.BaseColor, B=NamedRerouteUsage(), Alpha=@n13), B=@n11.BaseColor, Alpha=NamedRerouteUsage())
  LinearInterpolate(A=LinearInterpolate(A=@n11.Metallic, Alpha=@n13), B=@n11.Metallic, Alpha=NamedRerouteUsage())
  n14 = TextureSampleParameter2D() {group=1_Input Textures, parameterName=Rim_Normal, samplerType=SAMPLERTYPE_Normal, texture=/VehicleCosmetics/Textures/T_BlankNormal_B}
  NamedRerouteDeclaration(None=ConstantBiasScale(None=@n14.R) {bias=-0.5, scale=1}) {name=wheelnormal_R, nodeColor=<rgba 1 0.2812 0 1>}
  NamedRerouteDeclaration(None=ConstantBiasScale(None=@n14.G) {bias=-0.5, scale=1}) {name=wheelnormal_G, nodeColor=<rgba 1 0.4922 0 1>}
  n15 = TextureSampleParameter2D(UVs=Panner(Coordinate=TextureCoordinate() {uTiling=2, unMirrorU=true, vTiling=0.5}) {speedX=0.02}) {
   desc=DetailNormal
   group=1_Input Textures
   parameterName=DetailNormal
   samplerType=SAMPLERTYPE_Normal
   texture=/VehicleCosmetics/Wheels/RL209SE/Textures/T_RL209SE_
  }
  n16 = Saturate(None=@n15.RGB)
  n17 = Clamp(None=ConstantBiasScale(None=@n16) {bias=-0.5, scale=1})
  NamedRerouteDeclaration(None=Saturate(None=ComponentMask(None=@n17) {b=true})) {name=detailnormal_B, nodeColor=<rgba 0 1 0.1562 1>}
  NamedRerouteDeclaration(None=@n16) {name=DetailNormal_RGB, nodeColor=<rgba 0 1 0.1562 1>}
  n18 = ReflectionVectorWS() {bNormalizeCustomWorldNormal=true}
  n19 = TextureSampleParameter2D() {desc=tire_facetmask, group=1_Input Textures, parameterName=Facet_Texture, texture=/VehicleCosmetics/Wheels/RL209SE/Textures/T_RL209SE_facets}
  n20 = Normalize(VectorInput=Add(A=@n19.RGB, B=Constant3Vector() {constant=<rgba 0.3 0.3 1 0>}))
  n21 = Constant3Vector() {constant=<rgba 0.5 0.5 1 0>}
  n22 = NamedRerouteUsage()
  n23 = LinearInterpolate(
   A=Multiply(A=@n14.RGB, B=Constant3Vector() {constant=<rgba 2 2 1 0>})
   B=MaterialFunctionCall(Base=Add(A=@n20, B=@n20), Blend=Multiply(A=@n21, B=@n22)) {materialFunction=&Blend_Overlay}.Result
   Alpha=@n8.R
  )
  n24 = ConstantBiasScale(None=@n23) {bias=-0.5, scale=1}
  n25 = TextureSample(UVs=Rotator(Coordinate=Add(A=@n18, B=@n24)) {speed=0.5}) {samplerType=SAMPLERTYPE_LinearColor, texture=/VehicleCosmetics/Mutable/FX_Materials/WorldEnv/WorldEnv_AthenaCubemap}
  n26 = ScalarParameter() {defaultValue=1, parameterName=EnvironmentDesaturation}
  n27 = VectorParameter() {parameterName=CustomColor, defaultValue=<rgba 0.1799 0.1885 0.2188 1>}
  n28 = LinearInterpolate(A=@n27.RGB, B=Clamp(None=Multiply(A=@n27.RGB, B=@n27.RGB) {constB=50}), Alpha=Floor(None=ScalarParameter() {parameterName=DarkerCentre}))
  n29 = Constant() {r=0.08}
  n30 = TextureSampleParameter2D() {parameterName=Rim_MaskRGBA, samplerType=SAMPLERTYPE_LinearColor, texture=/VehicleCosmetics/Wheels/RL209/Textures/T_Rim_RL209_LOD_M}
  n31 = LinearInterpolate(
   A=LinearInterpolate(A=@n28, B=VectorParameter() {parameterName=Rim_ColorA, defaultValue=<rgba 0.5 0.2476 0.0599 1>}.RGB, Alpha=@n30.R)
   B=VectorParameter() {parameterName=Rim_ColorB, defaultValue=<rgba 1 0 0 1>}.RGB
   Alpha=@n30.G
  )
  n32 = VectorParameter() {parameterName=Rim_ColorC, defaultValue=<rgba 0 1 0.1405 1>}
  n33 = LinearInterpolate(A=@n32.RGB, B=Clamp(None=Multiply(A=@n32.RGB, B=@n32.RGB) {constB=50}), Alpha=Floor(None=ScalarParameter() {parameterName=DarkerCentre}))
  n34 = NamedRerouteUsage()
  n35 = LinearInterpolate(
   A=Multiply(A=@n28, B=@n29)
   B=StaticSwitchParameter(True=LinearInterpolate(A=@n31, B=@n33, Alpha=@n30.B), False=@n28) {defaultValue=true, parameterName=UseTertiaryColor}
   Alpha=@n34
  )
  n36 = VectorParameter() {parameterName=AccentColor, defaultValue=<rgba 0.3325 0.5647 1 1>}
  n37 = LinearInterpolate(A=Multiply(A=@n29, B=@n36.RGB), B=@n36.RGB, Alpha=@n34)
  n38 = Clamp(None=Add(A=NamedRerouteUsage(), B=NamedRerouteUsage()))
  n39 = VectorParameter() {parameterName=Tire_ColorA, defaultValue=<rgba 0.3325 0.5647 1 1>}
  n40 = Multiply(A=Constant() {r=0.2}, B=@n39.RGB)
  n41 = Clamp(None=Add(A=ConstantBiasScale(None=@n5.R) {bias=-0.5, scale=1}, B=ConstantBiasScale(None=@n5.G) {bias=-0.5, scale=1}))
  n42 = LinearInterpolate(A=@n40, B=@n39.RGB, Alpha=Add(A=@n41, B=@n41))
  n43 = Add(
   A=Multiply(A=Desaturation(None=@n25.RGB, Fraction=@n26), B=ScalarParameter() {defaultValue=0.7, parameterName=ReflectionIntensity})
   B=LinearInterpolate(A=LinearInterpolate(A=@n35, B=@n37, Alpha=@n38), B=LinearInterpolate(A=@n40, B=@n42, Alpha=@n9), Alpha=VertexColor().R)
  )
  n44 = Power(Base=Multiply(A=@n43, B=@n43), Exp=Constant() {r=0.7})
  NamedRerouteDeclaration(None=Clamp(None=@n44)) {name=Base_out, nodeColor=<rgba 0.0547 0 1 1>}
  n45 = VectorParameter() {parameterName=SpecColor, defaultValue=<rgba 0.7305 0.7305 0.7305 1>}
  n46 = ScalarParameter() {defaultValue=1, parameterName=SpecStrength}
  NamedRerouteDeclaration(None=Multiply(A=@n23, B=Multiply(A=@n45.RGB, B=@n46))) {name=spec_out, nodeColor=<rgba 0.0547 0 1 1>}
  NamedRerouteDeclaration(None=Clamp(None=Multiply(A=@n44))) {name=Emissive_out, nodeColor=<rgba 0.0547 0 1 1>}
  NamedRerouteDeclaration(None=@n23) {name=normal_out, nodeColor=<rgba 0.0547 0 1 1>}
  NamedRerouteDeclaration(None=VertexColor().R) {name=VC_R, nodeColor=<rgba 1 0 0 1>}
  NamedRerouteDeclaration(None=@n19.RGBA) {name=facets_a, nodeColor=<rgba 1 0 0 1>}
  Saturate(None=@n17)
  ScalarParameter() {defaultValue=1, parameterName=Darker_amount}
  Transform(None=@n18) {transformSourceType=TRANSFORMSOURCE_Local}
 ]}
